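; Dynamic UI director component for battle royale dual weapons that shows or hides an ammo UI scene based on a tagged inventory item.
; It binds to the inventory context's world item list changes when the UI is ready and unbinds on EndPlay.
; Each change checks the local player's inventory for the item tag, falling back to the view target's inventory for spectating.
; A delta gate adds or removes the scene only when that state changes.
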(event EventOnReadyToShowUI
  (Inventory|BindEventtoOnWorldItemListChanged (BlueprintContext|GetFortInventoryContext) (EventDispatchers|CreateEvent :!event OnWorldItemListChanged)))

(event EventEndPlay
  (params (EndPlayReason :type EEndPlayReason%20Enum :default "Destroyed"))
  (bind _returnvalue (BlueprintContext|GetFortInventoryContext))
  (Utilities|IsValid _returnvalue
    (:Is%20Valid
      (Inventory|UnbindEventfromOnWorldItemListChanged _returnvalue (EventDispatchers|CreateEvent :!event OnWorldItemListChanged)))))

(fn OnWorldItemListChanged
  (params (ItemsAdded :type Array%20of%20Fort%20World%20Item%20Object%20References) (ItemsRemoved :type Array%20of%20Fort%20World%20Item%20Object%20References))
  (bind _itemtag (Variables|Default|GetItemTag))
  (bind _asfort_player_controller (Utilities|Casting|CastToFortPlayerController (CallFunction|GetOwningLocalPlayerController)))
  (bind _returnvalue (Inventory|GetInventoryItemwithTag _asfort_player_controller _itemtag))
  (Utilities|IsValid _returnvalue
    (:Is%20Valid
      (CallFunction|SetAmmoSceneActive true))
    (:Is%20Not%20Valid
      (bind _returnvalue_2 (Inventory|GetViewTargetInventoryItemwithTag _asfort_player_controller _itemtag false)) ; For spectate
      (CallFunction|SetAmmoSceneActive (Utilities|IsValid _returnvalue_2 :!class /Script/Engine.KismetSystemLibrary)))))

(fn SetAmmoSceneActive
  (params (IsSceneActive :type Boolean))
  (bind _ammoscene (Variables|Default|GetAmmoScene))
  (bind _updated_variable (Utilities|FlowControl|DeltaGateWildcard (Variables|Default|GetIsAmmoSceneActive) IsSceneActive)
    (:Success
      (if _updated_variable
        (CallFunction|AddScene _ammoscene)
        (else
          (CallFunction|RemoveScene _ammoscene))))))
